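# Niagara swing effect for the Neon Graffiti pickaxe, leaving a short-lived glowing trail of mesh particles.
# A single emitter spawns by distance moved during a brief, delayed window, placing particles at randomly offset points sampled from a user-supplied skeletal mesh and matching their orientation.
# Particles inherit speed-limited velocity with drag.
# Very bright, low-alpha color fades out while an irregular RGB curve adds neon-like flicker.
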NiagaraSystem /BRCosmetics/Weapons/FORT_Melee/Pickaxe_NeonGraffiti/VFX/NS_Pickaxe_NeonGraffiti_Swing
dependencies {usedDataInterfaces [&NiagaraDataInterfaceCurve], usedDynamicInputs [], usedModules [&SystemState]}
emitters {Empty {
  emitterUpdateScript [
   &EmitterState(Life Cycle Mode=::NewEnumerator1, Loop Behavior=::NewEnumerator1, Loop Delay=0.07, Loop Duration=0.23, UseLoopDelay=true)
   &SpawnPerUnit(Module.Use Max Movement Threshold=false, Module.Use Movement Tolerance=true, Spawn Per Unit=45, Use Spawn Probability=false)
  ]
  particleSpawnScript [
   &InitializeParticle(
    Color=<rgba 50 50 50 0.025>
    Lifetime Max=0.15
    Lifetime Min=0.1
    Lifetime Mode=::NewEnumerator1
    Mesh Scale Mode=::NewEnumerator3
    UsePositionOffset=false
    Write Material Random=true
   )
   &SkeletalMeshLocation(
    Apply Sampled Orientation to Mesh Particles=true
    Inherit Velocity=false
    MaskSpawnGroup=false
    Mesh Sampling Type=::NewEnumerator5
    Sampled Position Offset=&RandomRangeVector(Minimum=<0 -2 8.75>, Maximum=<0 2 6.75>)
    Sampled Position Offset Coordinate Space:ENiagaraMeshTransforms=::NewEnumerator3
    Skeletal Mesh=$User.New Skeletal Mesh
    Socket Sampling Mode=::NewEnumerator5
   )
  ]
  particleUpdateScript [
   &ParticleState()
   &ScaleColor(
    Scale Alpha=&Multiply_Float(A=&FloatFromCurve(FloatCurve=DataInterface() {propertyValues {Curve {keys [<at 0 1>,<at 1 0>]}}}), B=1)
    Scale RGB=&VectorFromFloat(Value=&FloatFromCurve(FloatCurve=DataInterface() {propertyValues {Curve {keys [
          <at 0 1>
          <at 0.1 0.1>
          {arriveTangent=5.8125, leaveTangent=5.8125, time=0.12, value=1.03}
          {arriveTangent=0.9524, leaveTangent=0.9524, time=0.26, value=1.03}
          {arriveTangent=-2.6486, leaveTangent=-2.6486, time=0.54, value=1.43}
          {arriveTangent=-8.9375, leaveTangent=-8.9375, time=0.63, value=0.05}
          {arriveTangent=12.0833, interpMode=RCIM_Cubic, leaveTangent=12.0833, time=0.7, value=0}
          <at 0.75 1.5>
          <at 0.8 1.5>
          <at 0.9 0>
          <at 1 1.5>
         ]}}}))
    ScaleA=true
    ScaleRGB=true
   )
   &InheritVelocity(Module.Limit Speed=true, Velocity Limit=20)
   &Drag(Drag=5, Use Linear Drag=true, Use Rotational Drag=false)
   &SolveForcesAndVelocity(Clamp Velocity=false, Limit Acceleration=false)
  ]
  rendererClasses [&NiagaraMeshRendererProperties]
  renderers [{rendererClass=&NiagaraMeshRendererProperties, rendererIndex=0}]
 }}
summary {
 systemName=NS_Pickaxe_NeonGraffiti_Swing
 emitters [{emitterName=Empty, rendererClasses [&NiagaraMeshRendererProperties]}]
 userVariables [User.New Skeletal Mesh {
   type=NiagaraDataInterfaceSkeletalMesh
   defaultValue {
    struct=NiagaraExt_VariableValue_DataInterface
    value {dataInterface=&self:NiagaraDataInterfaceSkeletalMesh_0}
   }
  }]
}
